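# Niagara system for the LimeRock energy ball effect, with optional summon and ultra variants set by user bools.
# Mesh main and plasma spheres combine with light-casting glows, flickering mesh bolts and bright and dark streaks.
# System.Forming drives spawning, size and brightness, with curve-timed flares late in formation, while FX_Scalar scales the layers.
# IsSummon enables converging GPU dots, extra bolts and glow; IsUltra alters colors and material parameters.
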NiagaraSystem /LimeRock/Assets/FX/EnergyBall/NS_LimeRock_EnergyBall
dependencies {
 usedDataInterfaces [
  &NiagaraDataInterfaceCurve
  &NiagaraDataInterfaceArrayFloat
  &NiagaraDataInterfaceVector2DCurve
  &NiagaraDataInterfaceColorCurve
  &NiagaraDataInterfaceVectorCurve
 ]
 usedDynamicInputs [
  &RemapRange
  &VectorLength
  &NormalizeVector
 ]
 usedModules [
  &SystemState
  &Timeline
  &self:SystemSpawnScript.SystemScriptSource.SystemScriptGraph.NiagaraNodeAssignment_3.SetVariables_#1
  &self:SystemSpawnScript.SystemScriptSource.SystemScriptGraph.NiagaraNodeAssignment_2.SetVariables_#2
  &self:SystemSpawnScript.SystemScriptSource.SystemScriptGraph.NiagaraNodeAssignment_0.SetVariables_#3
  &self:SystemSpawnScript.SystemScriptSource.SystemScriptGraph.NiagaraNodeAssignment_1.SetVariables_#4
 ]
}
emitters {
 BlackDots {
  bEnabled=false
  simTarget=GPUComputeSim
  emitterUpdateScript [
   &EmitterState()
   &SpawnRate(Spawn Probability=$System.Forming, SpawnRate=50, Use Spawn Probability=true)
  ]
  particleSpawnScript [
   &InitializeParticle(
    Lifetime Max=0.35
    Lifetime Min=0.25
    Lifetime Mode=::NewEnumerator1
    Sprite Rotation Mode=::NewEnumerator2
    Sprite Size=<7 12>
    Sprite Size Mode=::NewEnumerator4
    Sprite UV Mode=::NewEnumerator3
    UsePositionOffset=true
    Write Material Random=true
   )
   &ShapeLocation(
    MaskSpawnGroup=false
    Override Seed=false
    Rotation Mode=::NewEnumerator1
    Sphere Radius=&Multiply_Float(A=50, B=$System.FX_Scalar)
    UseOwnerScale=true
   )
   &AddVelocity(
    Module.Use Falloff Distance=false
    Origin Offset=<-50 0 0>
    Override Seed=false
    Velocity Mode=::NewEnumerator1
    Velocity Speed=&Multiply_Float(
     A=&RandomRangeFloat(Minimum=0.1, Maximum=0.5)
     B=&Multiply_Float(
      A=&SelectFloatFromArray(Direct Array Index=$Particles.VisibilityTag, Float Selection Array=DataInterface() {propertyValues {FloatData [1,0.1]}})
      B=1600
     )
    )
   ) {enabled=false, module=AddVelocity002}
   &AddVelocity(Velocity=&Multiply_Vector(
     A=&MakeVector(
      X=&Multiply_Float(A=500, B=&RandomRangeFloat(Minimum=0.1, Maximum=1))
      Y=0
      Z=0
     )
     B=<1 1 1>
    )) {enabled=false}
   &AddVelocity(
    Module.Use Falloff Distance=false
    Override Seed=false
    Velocity Mode=::NewEnumerator1
    Velocity Speed=&Multiply_Float(A=300, B=$System.FX_Scalar)
   ) {module=AddVelocity001}
  ]
  particleUpdateScript [
   &ParticleState()
   &SpriteRotationRate(Rotation Rate=&Multiply_Float(A=&RandomRangeFloat(Minimum=-1, Maximum=1), B=800))
   &CurlNoiseForce(Noise Frequency=20, Noise Strength=2500, Pan Noise=false)
   &Drag(
    Drag=&RandomRangeFloat(Minimum=2, Maximum=5)
    Rotational Drag=7
    Use Linear Drag=true
    Use Rotational Drag=true
   )
   &ApplyOwnerScaleToAttributes(Scale Initial Sprite Size=true)
   &SolveForcesAndVelocity(Clamp Velocity=false, Limit Acceleration=false)
   &ScaleSpriteSize(
    Scale Sprite Size Mode=::NewEnumerator0
    Uniform Scale Factor=&Lerp_Float(A=0.5, B=2.5, Alpha=$System.DistanceScalar)
   ) {module=ScaleSpriteSize001}
   &ScaleSpriteSize(Scale Sprite Size Mode=::NewEnumerator0, Uniform Scale Factor=2)
   &ScaleSpriteSize(
    Uniform Curve Scale=2
    Uniform Curve Sprite Scale=DataInterface() {propertyValues {Curve {keys [
        {interpMode=RCIM_Cubic, tangentMode=RCTM_User, time=0, value=0}
        <at 0.7 0.95 RCIM_Cubic>
        {interpMode=RCIM_Cubic, tangentMode=RCTM_User, time=1, value=0}
       ]}}}
   ) {module=ScaleSpriteSize002}
   /SetVariables(Particles.SpriteSize=&Multiply_Vector2D_ByFloat(Float=$System.FX_Scalar, Vector2D=$Particles.SpriteSize)) {isSetParameters=true, module=SetVariables_#5}
  ]
  rendererClasses [&NiagaraSpriteRendererProperties]
  renderers [{rendererClass=&NiagaraSpriteRendererProperties}]
 }
 BlackDots001 {
  simTarget=GPUComputeSim
  emitterUpdateScript [
   &EmitterState(Inactive Response=::NewEnumerator1, Life Cycle Mode=::NewEnumerator1, Loop Behavior=::NewEnumerator1, Loop Duration=1, UseLoopDelay=false)
   &SpawnRate(
    Spawn Probability=&MakeCustomFloatFromBool(Bool=$User.IsSummon, True Float=1, False Float=0)
    SpawnRate=&Lerp_Float(A=150, B=1, Alpha=$System.Forming)
    Use Spawn Probability=true
   )
  ]
  particleSpawnScript [
   &InitializeParticle(
    Lifetime=&Lerp_Float(A=1, B=0.3, Alpha=$System.Forming)
    Sprite Rotation Mode=::NewEnumerator2
    Sprite Size=<7 12>
    Sprite Size Mode=::NewEnumerator4
    Sprite UV Mode=::NewEnumerator3
    UsePositionOffset=true
    Write Material Random=true
   )
   &ShapeLocation(
    MaskSpawnGroup=false
    Override Seed=false
    Rotation Mode=::NewEnumerator1
    Sphere Radius=&Multiply_Float(A=&Lerp_Float(A=900, B=250, Alpha=$System.Forming), B=$System.FX_Scalar)
    UseOwnerScale=true
   )
  ]
  particleUpdateScript [
   &ParticleState()
   &CurlNoiseForce(Noise Frequency=20, Noise Strength=8000, Pan Noise=false)
   &Drag(
    Drag=&RandomRangeFloat(Minimum=2, Maximum=5)
    Rotational Drag=7
    Use Linear Drag=true
    Use Rotational Drag=true
   )
   &ApplyOwnerScaleToAttributes(Scale Initial Sprite Size=true)
   &PointAttractionForce(
    Attraction Radius=1000
    Attraction Strength=&Lerp_Float(A=600, B=300, Alpha=$System.Forming)
    Kill Within Radius=false
    Module.Use Falloff=true
    UseAttractorOffset=false
   )
   &SolveForcesAndVelocity(Clamp Velocity=false, Limit Acceleration=false)
   &ScaleSpriteSize(
    Scale Sprite Size Mode=::NewEnumerator0
    Uniform Scale Factor=&Lerp_Float(A=0.5, B=2.5, Alpha=$System.DistanceScalar)
   ) {module=ScaleSpriteSize001}
   &ScaleSpriteSize(Scale Sprite Size Mode=::NewEnumerator0, Uniform Scale Factor=2)
   &ScaleSpriteSize(
    Uniform Curve Scale=2
    Uniform Curve Sprite Scale=DataInterface() {propertyValues {Curve {keys [
        {interpMode=RCIM_Cubic, tangentMode=RCTM_User, time=0, value=0}
        <at 0.2 0.95 RCIM_Cubic>
        {interpMode=RCIM_Cubic, tangentMode=RCTM_User, time=1, value=0}
       ]}}}
   ) {module=ScaleSpriteSize002}
   /SetVariables(Particles.SpriteSize=&Multiply_Vector2D_ByFloat(Float=$System.FX_Scalar, Vector2D=$Particles.SpriteSize)) {isSetParameters=true, module=SetVariables_#5}
  ]
  rendererClasses [&NiagaraSpriteRendererProperties]
  renderers [{rendererClass=&NiagaraSpriteRendererProperties}]
 }
 EnergySplineBolts {
  emitterUpdateScript [
   &EmitterState(Inactive Response=::NewEnumerator1, Life Cycle Mode=::NewEnumerator1, Loop Behavior=::NewEnumerator1, Loop Duration=0.55, UseLoopDelay=false)
   &SpawnBurst_Instantaneous(
    Spawn Count=6
    Spawn Probability=&MakeCustomFloatFromBool(Bool=$User.IsSummon, True Float=1, False Float=0)
    Spawn Time=0.5
    Use Spawn Probability=true
    UseLoopCountLimit=false
   ) {module=SpawnBurst_Instantaneous001}
   &SpawnRate(
    Spawn Probability=&MakeCustomFloatFromBool(Bool=$User.IsSummon, True Float=1, False Float=0)
    SpawnRate=&Lerp_Float(A=10, B=1, Alpha=$System.Forming)
    Use Spawn Probability=true
   )
   /SetVariables(Emitter.V_Scale_Random:NiagaraFloat=&RandomRangeFloat(Minimum=0.7, Maximum=1.1)) {isSetParameters=true, module=SetVariables_#6}
  ]
  particleSpawnScript [
   &InitializeParticle(
    Color=<rgba 0.0251 0.8368 1 1>
    Lifetime Max=0.8
    Lifetime Min=0.3
    Lifetime Mode=::NewEnumerator1
    Mesh Scale Mode=::NewEnumerator1
    Mesh Uniform Scale Max=15
    Mesh Uniform Scale Min=12
    UsePositionOffset=false
    Write Material Random=true
   )
   &InitialMeshOrientation(Mesh Orientation Mode=::NewEnumerator2, Use Rotation Vector=true)
   &CameraOffset(Camera Offset Amount=&Multiply_Float(A=200, B=$System.FX_Scalar))
   &ShapeLocation(
    MaskSpawnGroup=true
    Offset=<0 0 -150>
    Offset Coordinate Space=::World
    Offset Mode=::NewEnumerator0
    Override Seed=false
    Sphere Radius=&Multiply_Float(A=100, B=$System.FX_Scalar)
    Sphere Surface Distribution=1
    UseOwnerScale=true
   ) {module=ShapeLocation001}
   &AddVelocity(Module.Use Falloff Distance=false, Override Seed=false, Velocity Mode=::NewEnumerator1)
  ]
  particleUpdateScript [
   /SetVariables(Particles.Flicker:NiagaraFloat=&RandomRangeFloat(Minimum=0, Maximum=1, Evaluation Type=::NewEnumerator1)) {isSetParameters=true, module=SetVariables_#7}
   &ParticleState()
   &ScaleMeshSize(Scale Factor=&VectorFromCurve(
     VectorCurve=DataInterface() {propertyValues {
       XCurve {keys [{arriveTangent=0.4178, interpMode=RCIM_Cubic, leaveTangent=0.4178, tangentMode=RCTM_User, time=0, value=1}, <at 1 1.2 RCIM_Cubic>]}
       YCurve {keys [{arriveTangent=1.8574, interpMode=RCIM_Cubic, leaveTangent=1.8574, tangentMode=RCTM_User, time=0, value=1}, <at 1 2 RCIM_Cubic>]}
       ZCurve {keys [<at 0 1>,<at 1 1>]}
      }}
     Scale Curve=&VectorFromFloat(Value=1.6)
    )) {module=ScaleMeshSize001}
   &ScaleMeshSize(Scale Factor=&Multiply_VectorByFloat(Vector=<1 1 0.5>, Float=$System.FX_Scalar))
   &DynamicMaterialParameters(
    Index 0 Param 1=1
    Index 0 Param 2=5
    Index 1 Param 1=&Multiply_Float(
     A=&Multiply_Float(
      A=&Lerp_Float(A=30, B=50, Alpha=$System.Forming)
      B=&MakeFloatFromVector(Vector=$Engine.Owner.Scale, Channel=::NewEnumerator0)
     )
     B=&RandomRangeFloat(Minimum=0.2, Maximum=1)
    )
    Index 1 Param 2=1
    Index 1 Param 3=&RandomRangeFloat(Minimum=0.8, Maximum=1.8)
    Index 1 Param 4=1
    Index 2 Param 1=&FloatFromCurve(
     FloatCurve=DataInterface() {propertyValues {Curve {keys [
         {arriveTangent=0.6019, interpMode=RCIM_Cubic, leaveTangent=0.6019, tangentMode=RCTM_User, time=-0.0019, value=0.1}
         {arriveTangent=0.6592, interpMode=RCIM_Cubic, leaveTangent=0.6592, tangentMode=RCTM_User, time=0.4, value=0.4}
         {arriveTangent=0.2189, interpMode=RCIM_Cubic, leaveTangent=0.2189, tangentMode=RCTM_User, time=1, value=0.6}
        ]}}}
     Scale Curve=0.5
    )
    Module.Param0WriteEnabled=true
    Module.Param0WriteEnabled1=true
    Module.Param0WriteEnabled2=true
    Module.Param1WriteEnabled=true
    Module.Param1WriteEnabled1=true
    Module.Param1WriteEnabled2=true
    Module.Param2WriteEnabled=true
    Module.Param2WriteEnabled1=true
    Module.Param2WriteEnabled2=true
    Module.Param3WriteEnabled=true
    Module.Param3WriteEnabled1=true
    Module.Param3WriteEnabled2=true
    Write Parameter Index 1=true
    Write Parameter Index 2=true
   )
   &Drag(Drag=3, Use Linear Drag=true, Use Rotational Drag=true)
   &SolveForcesAndVelocity(Clamp Velocity=false, Limit Acceleration=false)
   &ScaleColor(Scale RGB=&VectorFromFloat(Value=5), ScaleA=true, ScaleRGB=true) {module=ScaleColor002}
   &ScaleColor(
    Scale Alpha=&Lerp_Float(A=0.3, B=1, Alpha=$Particles.Flicker)
    Scale RGB=&VectorFromFloat(Value=&Multiply_Float(
      A=&Lerp_Float(A=1, B=2, Alpha=$Particles.Flicker)
      B=&FloatFromCurve(
       FloatCurve=DataInterface() {propertyValues {Curve {keys [{interpMode=RCIM_Cubic, tangentMode=RCTM_User, time=0, value=1}, {interpMode=RCIM_Cubic, tangentMode=RCTM_User, time=1, value=0}]}}}
       Scale Curve=0.1
      )
     ))
    ScaleA=true
    ScaleRGB=true
   ) {module=ScaleColor001}
  ]
  rendererClasses [&NiagaraMeshRendererProperties]
  renderers [{rendererClass=&NiagaraMeshRendererProperties, rendererIndex=0}]
 }
 EnergySplineBolts001 {
  emitterUpdateScript [
   &EmitterState(
    Inactive Response=::NewEnumerator1
    Life Cycle Mode=::NewEnumerator1
    Loop Delay=&RandomRangeFloat(Minimum=0.05, Maximum=0.3)
    Loop Duration=&RandomRangeFloat(Minimum=0.2, Maximum=0.3)
    UseLoopDelay=true
   )
   &SpawnBurst_Instantaneous(
    Spawn Count=&RandomRangeInt(Minimum=1, Maximum=2)
    Spawn Time=0.1
    Use Spawn Probability=false
    UseLoopCountLimit=false
   )
   /SetVariables(Emitter.V_Scale_Random:NiagaraFloat=&RandomRangeFloat(Minimum=0.7, Maximum=1.1)) {isSetParameters=true, module=SetVariables_#6}
  ]
  particleSpawnScript [
   &InitializeParticle(
    Color=&MakeCustomLinearColorFromBool(Bool=$User.IsUltra, True Color=<rgba 0.0855 0.1069 1 1>, False Color=<rgba 0.0832 0.4003 1 1>)
    Lifetime Max=0.8
    Lifetime Min=0.3
    Lifetime Mode=::NewEnumerator1
    Mesh Scale Mode=::NewEnumerator1
    Mesh Uniform Scale Max=15
    Mesh Uniform Scale Min=12
    UsePositionOffset=false
    Write Material Random=true
   )
   &InitialMeshOrientation(Mesh Orientation Mode=::NewEnumerator2, Use Rotation Vector=true)
   &InitialMeshOrientation(Use Rotation Vector=true) {enabled=false, module=InitialMeshOrientation001}
   &CameraOffset(Camera Offset Amount=&Multiply_Float(A=200, B=$System.FX_Scalar))
   &ShapeLocation(
    MaskSpawnGroup=true
    Override Seed=false
    Sphere Radius=&Multiply_Float(A=100, B=$System.FX_Scalar)
    Sphere Surface Distribution=1
    UseOwnerScale=true
   ) {module=ShapeLocation001}
   &AddVelocity(Module.Use Falloff Distance=false, Override Seed=false, Velocity Mode=::NewEnumerator1)
  ]
  particleUpdateScript [
   /SetVariables(Particles.Flicker:NiagaraFloat=&RandomRangeFloat(Minimum=0, Maximum=1, Evaluation Type=::NewEnumerator1)) {isSetParameters=true, module=SetVariables_#7}
   &ParticleState()
   &ScaleMeshSize(Scale Factor=&VectorFromCurve(
     VectorCurve=DataInterface() {propertyValues {
       XCurve {keys [{arriveTangent=0.4178, interpMode=RCIM_Cubic, leaveTangent=0.4178, tangentMode=RCTM_User, time=0, value=1}, <at 1 1.2 RCIM_Cubic>]}
       YCurve {keys [{arriveTangent=1.8574, interpMode=RCIM_Cubic, leaveTangent=1.8574, tangentMode=RCTM_User, time=0, value=1}, <at 1 2 RCIM_Cubic>]}
       ZCurve {keys [<at 0 1>,<at 1 1>]}
      }}
     Scale Curve=&VectorFromFloat(Value=1.6)
    )) {module=ScaleMeshSize001}
   &ScaleMeshSize(Scale Factor=&Multiply_VectorByFloat(Vector=<1 1 0.5>, Float=$System.FX_Scalar))
   &DynamicMaterialParameters(
    Index 0 Param 1=1
    Index 0 Param 2=5
    Index 1 Param 1=&Multiply_Float(
     A=&Multiply_Float(A=50, B=&MakeFloatFromVector(Vector=$Engine.Owner.Scale, Channel=::NewEnumerator0))
     B=&RandomRangeFloat(Minimum=0.2, Maximum=1)
    )
    Index 1 Param 2=1
    Index 1 Param 3=&RandomRangeFloat(Minimum=0.8, Maximum=1.8)
    Index 1 Param 4=1
    Index 2 Param 1=&FloatFromCurve(
     FloatCurve=DataInterface() {propertyValues {Curve {keys [
         {arriveTangent=0.6019, interpMode=RCIM_Cubic, leaveTangent=0.6019, tangentMode=RCTM_User, time=-0.0019, value=0.1}
         {arriveTangent=0.6592, interpMode=RCIM_Cubic, leaveTangent=0.6592, tangentMode=RCTM_User, time=0.4, value=0.4}
         {arriveTangent=0.2189, interpMode=RCIM_Cubic, leaveTangent=0.2189, tangentMode=RCTM_User, time=1, value=0.6}
        ]}}}
     Scale Curve=0.5
    )
    Module.Param0WriteEnabled=true
    Module.Param0WriteEnabled1=true
    Module.Param0WriteEnabled2=true
    Module.Param1WriteEnabled=true
    Module.Param1WriteEnabled1=true
    Module.Param1WriteEnabled2=true
    Module.Param2WriteEnabled=true
    Module.Param2WriteEnabled1=true
    Module.Param2WriteEnabled2=true
    Module.Param3WriteEnabled=true
    Module.Param3WriteEnabled1=true
    Module.Param3WriteEnabled2=true
    Write Parameter Index 1=true
    Write Parameter Index 2=true
   )
   &Drag(Drag=3, Use Linear Drag=true, Use Rotational Drag=true)
   &SolveForcesAndVelocity(Clamp Velocity=false, Limit Acceleration=false)
   &ScaleColor(Scale RGB=&VectorFromFloat(Value=5), ScaleA=true, ScaleRGB=true) {module=ScaleColor002}
   &ScaleColor(
    Scale Alpha=&Lerp_Float(A=0.3, B=1, Alpha=$Particles.Flicker)
    Scale RGB=&VectorFromFloat(Value=&Multiply_Float(
      A=&Lerp_Float(A=500, B=1, Alpha=$Particles.Flicker)
      B=&FloatFromCurve(
       FloatCurve=DataInterface() {propertyValues {Curve {keys [{interpMode=RCIM_Cubic, tangentMode=RCTM_User, time=0, value=1}, {interpMode=RCIM_Cubic, tangentMode=RCTM_User, time=1, value=0}]}}}
       Scale Curve=0.1
      )
     ))
    ScaleA=true
    ScaleRGB=true
   ) {module=ScaleColor001}
  ]
  rendererClasses [&NiagaraMeshRendererProperties]
  renderers [{rendererClass=&NiagaraMeshRendererProperties, rendererIndex=0}]
 }
 LightAndGlow {
  emitterUpdateScript [
   &EmitterState(
    Inactive Response=::NewEnumerator1
    Life Cycle Mode=::NewEnumerator1
    Loop Behavior=::NewEnumerator1
    Loop Duration Mode=::NewEnumerator1
    UseLoopDelay=false
   )
   &SpawnBurst_Instantaneous(Use Spawn Probability=false, UseLoopCountLimit=true)
  ]
  particleSpawnScript [
   /SetVariables(Particles.VisibilityTag=$Emitter.SpawnGroup) {isSetParameters=true}
   &InitializeParticle(
    Color=&MakeCustomLinearColorFromBool(Bool=$User.IsUltra, True Color=<rgba 0.6956 0.3391 5 1>, False Color=<rgba 0.6956 0.3391 5 1>)
    Sprite Size Mode=::NewEnumerator3
    Uniform Sprite Size=800
    UsePositionOffset=false
    Write Material Random=true
   )
   &DynamicMaterialParameters(
    Index 0 Param 1=&MakeCustomFloatFromBool(Bool=$User.IsUltra, True Float=1, False Float=0)
    Module.Index0Param0Write=true
    Module.Index0Param1Write=true
    Module.Index0Param2Write=true
    Module.Index0Param3Write=true
   )
   &CameraOffset(Camera Offset Amount=&Multiply_Float(A=50, B=$System.FX_Scalar))
  ]
  particleUpdateScript [
   &ParticleState(Kill Particles When Lifetime Has Elapsed=false, Let Infinitely Lived Particles Die When Emitter Deactivates=true)
   &FN_Light_Attributes(
    Color=&MultiplyLinearColor(Linear Color=<rgba 0.4062 1 0.0184 1>, Scale Factor=&MakeVector4_FromFloat(Float=20))
    Create Color Attribute=true
    Create Radius Attribute=true
    Enable Particle Light Enable=true
    Light Enabled=true
    Radius=&Multiply_Float(
     A=400
     B=&Multiply_Float(A=&MakeFloatFromVector(Vector=$Engine.Owner.Scale, Channel=::NewEnumerator0), B=$System.FX_Scalar)
    )
    Scale Color=&Lerp_Float(A=0, B=1, Alpha=$System.Forming)
   )
   &ApplyOwnerScaleToAttributes(Scale Initial Sprite Size=true)
   &SolveForcesAndVelocity(Clamp Velocity=false, Limit Acceleration=false)
   &ScaleColor(Scale Alpha=&Lerp_Float(A=0, B=0.3, Alpha=$System.Forming), ScaleA=true, ScaleRGB=true)
   /SetVariables(Particles.SpriteSize=&Multiply_Vector2D_ByFloat(Float=$System.FX_Scalar, Vector2D=$Particles.SpriteSize)) {isSetParameters=true, module=SetVariables_#5}
   /SetVariables(
    Particles.FrontGlowColor:LinearColor=&MultiplyLinearColor(
     Linear Color=&MakeCustomLinearColorFromBool(Bool=$User.IsUltra, True Color=<rgba 2 0.9405 0.068 1>, False Color=<rgba 2 0.9137 0.1111 1>)
     Scale Factor=&MakeVector4(X=1, Y=1, Z=1, W=&Lerp_Float(A=0, B=0.3, Alpha=$System.Forming))
    )
    Particles.FrontGlowSize:Vector2f=&Multiply_Vector2D(
     A=$Particles.SpriteSize
     B=&Multiply_Vector2D_ByFloat(Float=&MakeCustomFloatFromBool(Bool=$User.IsUltra, True Float=0.45, False Float=0.6), Vector2D=<1 1>)
    )
   ) {isSetParameters=true, module=SetVariables_#8}
  ]
  rendererClasses [&NiagaraSpriteRendererProperties, &NiagaraLightRendererProperties]
  renderers [
   {rendererClass=&NiagaraSpriteRendererProperties}
   {rendererClass=&NiagaraLightRendererProperties, rendererIndex=1}
   {rendererClass=&NiagaraSpriteRendererProperties, rendererIndex=2}
  ]
 }
 LightAndGlow001 {
  emitterUpdateScript [
   &EmitterState(Inactive Response=::NewEnumerator1, Life Cycle Mode=::NewEnumerator1, Loop Behavior=::NewEnumerator1, Loop Duration=1, UseLoopDelay=false)
   &SpawnRate(
    Spawn Probability=&MakeCustomFloatFromBool(Bool=$User.IsSummon, True Float=1, False Float=0)
    SpawnRate=8
    Use Spawn Probability=true
   )
  ]
  particleSpawnScript [
   &InitializeParticle(
    Color=<rgba 0.0549 0.7215 1 1>
    Lifetime=0.2
    Sprite Size Mode=::NewEnumerator3
    Uniform Sprite Size=&Multiply_Float(A=&Lerp_Float(A=2000, B=800, Alpha=$System.Forming), B=0.5)
    UsePositionOffset=false
    Write Material Random=true
   )
   &CameraOffset(Camera Offset Amount=&Multiply_Float(A=50, B=$System.FX_Scalar))
   &DynamicMaterialParameters(
    Index 0 Param 1=1
    Module.Index0Param0Write=true
    Module.Index0Param1Write=true
    Module.Index0Param2Write=true
    Module.Index0Param3Write=true
   )
  ]
  particleUpdateScript [
   &ParticleState()
   &FN_Light_Attributes(
    Color=&MultiplyLinearColor(Linear Color=<rgba 0.4062 1 0.0184 1>, Scale Factor=&MakeVector4_FromFloat(Float=20))
    Create Color Attribute=true
    Create Radius Attribute=true
    Enable Particle Light Enable=true
    Light Enabled=true
    Radius=&Multiply_Float(
     A=&Lerp_Float(A=800, B=400, Alpha=$System.Forming)
     B=&Multiply_Float(A=&MakeFloatFromVector(Vector=$Engine.Owner.Scale, Channel=::NewEnumerator0), B=$System.FX_Scalar)
    )
    Scale Color=&Lerp_Float(A=0, B=1, Alpha=$System.Forming)
   )
   &ApplyOwnerScaleToAttributes(Scale Initial Sprite Size=true)
   &SolveForcesAndVelocity(Clamp Velocity=false, Limit Acceleration=false)
   &ScaleColor(
    Scale Alpha=&Lerp_Float(A=0, B=0.3, Alpha=$System.Forming)
    Scale RGB=&VectorFromFloat(Value=&Lerp_Float(A=8, B=0, Alpha=$System.Forming))
    ScaleA=true
    ScaleRGB=true
   )
   /SetVariables(Particles.SpriteSize=&Multiply_Vector2D_ByFloat(
     Float=&Multiply_Float(
      A=$System.FX_Scalar
      B=&Lerp_Float(
       A=1
       B=2
       Alpha=&FloatFromCurve(
        FloatCurve=DataInterface() {propertyValues {Curve {keys [
            {interpMode=RCIM_Cubic, tangentMode=RCTM_User, time=0.75, value=0}
            {interpMode=RCIM_Cubic, tangentMode=RCTM_User, time=0.9, value=1}
            {tangentMode=RCTM_User, time=1, value=0}
           ]}}}
        CurveIndex=$System.Forming
       )
      )
     )
     Vector2D=$Particles.SpriteSize
    )) {isSetParameters=true, module=SetVariables_#5}
   /SetVariables(Particles.FrontGlowSize:Vector2f=&Multiply_Vector2D(A=$Particles.SpriteSize, B=<2 0.3>)) {isSetParameters=true, module=SetVariables_#8}
   &ScaleColor(
    Scale RGB=&VectorFromFloat(Value=&Lerp_Float(
      A=1
      B=2
      Alpha=&FloatFromCurve(
       FloatCurve=DataInterface() {propertyValues {Curve {keys [
           {interpMode=RCIM_Cubic, tangentMode=RCTM_User, time=0.7, value=0}
           {interpMode=RCIM_Cubic, tangentMode=RCTM_User, time=0.9, value=1}
           {tangentMode=RCTM_User, time=1, value=0}
          ]}}}
       CurveIndex=$System.Forming
      )
     ))
    ScaleA=true
    ScaleRGB=true
   ) {module=ScaleColor001}
  ]
  rendererClasses [&NiagaraSpriteRendererProperties, &NiagaraLightRendererProperties]
  renderers [
   {rendererClass=&NiagaraSpriteRendererProperties}
   {rendererClass=&NiagaraLightRendererProperties, rendererIndex=1}
   {rendererClass=&NiagaraSpriteRendererProperties, rendererIndex=2}
  ]
 }
 MainSphere {
  emitterUpdateScript [
   &EmitterState(
    Inactive Response=::NewEnumerator1
    Life Cycle Mode=::NewEnumerator1
    Loop Behavior=::NewEnumerator1
    Loop Duration Mode=::NewEnumerator1
    UseLoopDelay=false
   )
   &SpawnBurst_Instantaneous(Use Spawn Probability=false, UseLoopCountLimit=false)
  ]
  particleSpawnScript [
   &InitializeParticle(Lifetime=1.5, Mesh Scale Mode=::NewEnumerator3, Mesh Uniform Scale=4, UsePositionOffset=false, Write Material Random=true)
   &InitialMeshOrientation(Mesh Axis To Orient=<0 0 1>, Mesh Orientation Mode=::NewEnumerator2, Use Rotation Vector=false)
  ]
  particleUpdateScript [
   &ParticleState(Kill Particles When Lifetime Has Elapsed=false, Let Infinitely Lived Particles Die When Emitter Deactivates=true)
   &ScaleMeshSize(Scale Factor=&VectorFromFloat(Value=$System.FX_Scalar)) {module=ScaleMeshSize001}
   &DynamicMaterialParameters(
    Index 0 Param 1=&Lerp_Float(A=1, B=0, Alpha=$System.Forming)
    Index 0 Param 2=1
    Module.Index0Param0Write=true
    Module.Index0Param1Write=true
    Module.Index0Param2Write=true
    Module.Index0Param3Write=true
   )
   &ScaleColor(
    Scale RGB=&VectorFromFloat(Value=&Lerp_Float(
      A=1
      B=8
      Alpha=&FloatFromCurve(
       FloatCurve=DataInterface() {propertyValues {Curve {keys [
           {interpMode=RCIM_Cubic, tangentMode=RCTM_User, time=0.85, value=0}
           {interpMode=RCIM_Cubic, tangentMode=RCTM_User, time=0.95, value=1}
           {tangentMode=RCTM_User, time=1, value=0}
          ]}}}
       CurveIndex=$System.Forming
      )
     ))
    ScaleA=true
    ScaleRGB=true
   )
   &ScaleColor(
    Scale RGB=&VectorFromFloat(Value=&MakeCustomFloatFromBool(
      Bool=$User.IsSummon
      True Float=1
      False Float=&Lerp_Float(
       A=1
       B=20
       Alpha=&FloatFromCurve(
        FloatCurve=DataInterface() {propertyValues {Curve {keys [{arriveTangent=-3.0621, interpMode=RCIM_Cubic, leaveTangent=-3.0621, tangentMode=RCTM_User, time=0, value=1}, <at 0.4 0 RCIM_Cubic>]}}}
        CurveIndex=$Emitter.Age
       )
      )
     ))
    ScaleA=true
    ScaleRGB=true
   ) {module=ScaleColor001}
  ]
  rendererClasses [&NiagaraMeshRendererProperties]
  renderers [{rendererClass=&NiagaraMeshRendererProperties, rendererIndex=0}]
 }
 PlasmaSphere {
  emitterUpdateScript [
   &EmitterState(
    Inactive Response=::NewEnumerator1
    Life Cycle Mode=::NewEnumerator1
    Loop Behavior=::NewEnumerator1
    Loop Duration Mode=::NewEnumerator1
    UseLoopDelay=false
   )
   &SpawnBurst_Instantaneous(Spawn Count=1, Use Spawn Probability=false)
  ]
  particleSpawnScript [&InitializeParticle(
    Color=<rgba 0.0008 0.1365 2 1>
    Lifetime=0.2
    Mesh Scale Mode=::NewEnumerator3
    Mesh Uniform Scale=3.5
    UsePositionOffset=false
    Write Material Random=true
   )]
  particleUpdateScript [
   &ParticleState(Kill Particles When Lifetime Has Elapsed=false, Let Infinitely Lived Particles Die When Emitter Deactivates=true)
   &ScaleMeshSize(Scale Factor=&VectorFromFloat(Value=$System.FX_Scalar))
   &UpdateMeshOrientation(Rotation Rate=0.2, Rotation Vector=<1 1 1>)
   &DynamicMaterialParameters(
    Index 0 Param 1=&Lerp_Float(A=1, B=0, Alpha=$System.Forming)
    Module.Param0WriteEnabled=true
    Module.Param1WriteEnabled=true
    Module.Param2WriteEnabled=true
    Module.Param3WriteEnabled=true
   )
   &ScaleColor(
    Scale RGB=&VectorFromFloat(Value=&Lerp_Float(
      A=20
      B=300
      Alpha=&FloatFromCurve(
       FloatCurve=DataInterface() {propertyValues {Curve {keys [
           {interpMode=RCIM_Cubic, tangentMode=RCTM_User, time=0.77, value=0}
           {interpMode=RCIM_Cubic, tangentMode=RCTM_User, time=0.8, value=1}
           {tangentMode=RCTM_User, time=1, value=0}
          ]}}}
       CurveIndex=$System.Forming
      )
     ))
    ScaleA=true
    ScaleRGB=true
   )
   &ScaleColor(
    Scale RGB=&VectorFromFloat(Value=&MakeCustomFloatFromBool(
      Bool=$User.IsSummon
      True Float=1
      False Float=&Lerp_Float(
       A=1
       B=200
       Alpha=&FloatFromCurve(
        FloatCurve=DataInterface() {propertyValues {Curve {keys [{arriveTangent=-3.0621, interpMode=RCIM_Cubic, leaveTangent=-3.0621, tangentMode=RCTM_User, time=0, value=1}, <at 0.4 0 RCIM_Cubic>]}}}
        CurveIndex=$Emitter.Age
       )
      )
     ))
    ScaleA=true
    ScaleRGB=true
   ) {enabled=false, module=ScaleColor001}
  ]
  rendererClasses [&NiagaraMeshRendererProperties]
  renderers [{rendererClass=&NiagaraMeshRendererProperties, rendererIndex=0}]
 }
 StreaksBright {
  emitterSpawnScript [/SetVariables(Emitter.FormingCurve=DataInterface() {propertyValues {Curve {keys [
        {interpMode=RCIM_Cubic, tangentMode=RCTM_User, time=0.7, value=0}
        {interpMode=RCIM_Cubic, tangentMode=RCTM_User, time=0.8, value=1}
        {tangentMode=RCTM_User, time=1, value=0}
       ]}}}) {isSetParameters=true, module=SetVariables_#9}]
  emitterUpdateScript [
   &EmitterState()
   &SpawnRate(
    Spawn Probability=$System.Forming
    SpawnRate=&Lerp_Float(
     A=25
     B=70
     Alpha=&Multiply_Float(
      A=1
      B=&Lerp_Float(A=1, B=2.5, Alpha=&FloatFromCurve(FloatCurve=$Emitter.FormingCurve, CurveIndex=$System.Forming))
     )
    )
    Use Spawn Probability=true
   )
  ]
  particleSpawnScript [
   &PartitionParticles()
   &InitializeParticle(
    Lifetime Max=0.5
    Lifetime Min=0.4
    Lifetime Mode=::NewEnumerator1
    Sprite Size=&MakeVector2D(
     X=&Multiply_Float(
      A=35
      B=&Lerp_Float(A=1, B=2.5, Alpha=&FloatFromCurve(FloatCurve=$Emitter.FormingCurve, CurveIndex=$System.Forming))
     )
     Y=&Multiply_Float(
      A=35
      B=&Lerp_Float(
       A=1
       B=4
       Alpha=&FloatFromCurve(
        FloatCurve=DataInterface() {propertyValues {Curve {keys [
            {interpMode=RCIM_Cubic, tangentMode=RCTM_User, time=0, value=0}
            {interpMode=RCIM_Cubic, tangentMode=RCTM_User, time=0.5, value=1}
            {tangentMode=RCTM_User, time=1, value=0}
           ]}}}
        CurveIndex=$System.Forming
       )
      )
     )
    )
    Sprite Size Mode=::NewEnumerator4
    Sprite UV Mode=::NewEnumerator1
    UsePositionOffset=false
    Write Material Random=true
   )
   &CameraOffset(Camera Offset Amount=&SelectFloatFromArray(Direct Array Index=&SpawnGroup(), Float Selection Array=DataInterface() {propertyValues {FloatData [10,10]}}))
   &ShapeLocation(
    MaskSpawnGroup=false
    Override Seed=false
    Sphere Radius=&Multiply_Float(A=50, B=$System.FX_Scalar)
    UseOwnerScale=true
   )
   &AddVelocity(
    Module.Use Falloff Distance=false
    Override Seed=false
    Velocity Mode=::NewEnumerator1
    Velocity Speed=&Multiply_Float(
     A=&Multiply_Float(
      A=200
      B=&Lerp_Float(A=1, B=7, Alpha=&FloatFromCurve(FloatCurve=$Emitter.FormingCurve, CurveIndex=$System.Forming))
     )
     B=$System.FX_Scalar
    )
   )
  ]
  particleUpdateScript [
   &ParticleState()
   &ScaleSpriteSize(Scale Sprite Size Mode=::NewEnumerator0)
   &ScaleSpriteSize(
    Non-Uniform Curve Scale=<2 5>
    Non-Uniform Curve Sprite Scale=DataInterface() {propertyValues {
      XCurve {keys [
        {arriveTangent=0.8609, interpMode=RCIM_Cubic, leaveTangent=0.8609, tangentMode=RCTM_User, time=0, value=0}
        {arriveTangent=0.5103, interpMode=RCIM_Cubic, leaveTangent=0.5103, tangentMode=RCTM_User, time=0.253, value=0.4867}
        {arriveTangent=0.0224, interpMode=RCIM_Cubic, leaveTangent=0.0224, tangentMode=RCTM_User, time=1, value=1}
       ]}
      YCurve {keys [
        {arriveTangent=0.8801, interpMode=RCIM_Cubic, leaveTangent=0.8801, tangentMode=RCTM_User, time=0, value=0.3}
        {arriveTangent=0.6586, interpMode=RCIM_Cubic, leaveTangent=0.6586, tangentMode=RCTM_User, time=0.2, value=0.5745}
        {arriveTangent=-0.4778, interpMode=RCIM_Cubic, leaveTangent=-0.4778, tangentMode=RCTM_User, time=1, value=1.05}
       ]}
     }}
    Scale Sprite Size Mode=::NewEnumerator3
   ) {module=ScaleSpriteSize001}
   &Drag(Drag=4, Use Linear Drag=true, Use Rotational Drag=true)
   &ApplyOwnerScaleToAttributes(Scale Initial Sprite Size=true)
   &SolveForcesAndVelocity(Clamp Velocity=false, Limit Acceleration=false)
   &ScaleColor(
    Scale Alpha=&Multiply_Float(
     A=&FloatFromCurve(
      FloatCurve=DataInterface() {propertyValues {Curve {keys [
          {arriveTangent=9.0725, interpMode=RCIM_Cubic, leaveTangent=9.0725, tangentMode=RCTM_User, time=0, value=1}
          {arriveTangent=-2.5249, interpMode=RCIM_Cubic, leaveTangent=-2.5249, tangentMode=RCTM_User, time=0.1408, value=0.9847}
          {arriveTangent=-0.6199, interpMode=RCIM_Cubic, leaveTangent=-0.6199, time=0.3, value=0.95}
          {arriveTangent=-1.3571, interpMode=RCIM_Cubic, leaveTangent=-1.3571, time=0.6, value=0.7}
          {arriveTangent=-0.0405, interpMode=RCIM_Cubic, leaveTangent=-0.0405, tangentMode=RCTM_User, time=1, value=0}
         ]}}}
      Scale Curve=5
     )
     B=1
    )
    Scale RGB=&VectorFromFloat(Value=1.5)
    ScaleA=true
    ScaleRGB=true
   ) {module=ScaleColor001}
   &ScaleColor(
    Scale Mode=::NewEnumerator1
    Scale RGBA=&MakeFloat4FromVectorAndFloat("Vector (RGB)"=<1 1 1>, "Float (Alpha)"=1)
    ScaleRGBA=true
   ) {enabled=false, module=ScaleColor002}
   &DynamicMaterialParameters(
    Index 0 Param 1=&FloatFromCurve(FloatCurve=DataInterface() {propertyValues {Curve {keys [
         {arriveTangent=-0.1846, time=0, value=0}
         {arriveTangent=1.3385, interpMode=RCIM_Cubic, leaveTangent=0.1195, tangentMode=RCTM_Break, time=0.3, value=0}
         {arriveTangent=1.4286, interpMode=RCIM_Cubic, leaveTangent=1.4286, time=0.8, value=0.5}
         {arriveTangent=3.0091, interpMode=RCIM_Cubic, leaveTangent=3.0091, tangentMode=RCTM_User, time=1, value=1}
        ]}}})
    Index 0 Param 2=&MakeCustomFloatFromBool(Bool=$User.IsUltra, True Float=0, False Float=1)
    Module.Index0Param0Write=true
    Module.Index0Param1Write=true
    Module.Index0Param2Write=false
    Module.Index0Param3Write=false
   )
   /SetVariables(Particles.SpriteSize=&Multiply_Vector2D_ByFloat(Float=$System.FX_Scalar, Vector2D=$Particles.SpriteSize)) {isSetParameters=true, module=SetVariables_#5}
  ]
  rendererClasses [&NiagaraSpriteRendererProperties]
  renderers [{rendererClass=&NiagaraSpriteRendererProperties}]
 }
 StreaksDark {
  emitterSpawnScript [/SetVariables(Emitter.FormingCurve=DataInterface() {propertyValues {Curve {keys [
        {interpMode=RCIM_Cubic, tangentMode=RCTM_User, time=0.7, value=0}
        {interpMode=RCIM_Cubic, tangentMode=RCTM_User, time=0.8, value=1}
        {tangentMode=RCTM_User, time=1, value=0}
       ]}}}) {isSetParameters=true, module=SetVariables_#9}]
  emitterUpdateScript [
   &EmitterState()
   &SpawnRate(
    Spawn Probability=$System.Forming
    SpawnRate=&Lerp_Float(
     A=18
     B=30
     Alpha=&Multiply_Float(
      A=1
      B=&Lerp_Float(A=1, B=2.5, Alpha=&FloatFromCurve(FloatCurve=$Emitter.FormingCurve, CurveIndex=$System.Forming))
     )
    )
    Use Spawn Probability=true
   )
  ]
  particleSpawnScript [
   &PartitionParticles()
   &InitializeParticle(
    Lifetime Max=1
    Lifetime Min=0.8
    Lifetime Mode=::NewEnumerator1
    Sprite Size=&MakeVector2D(
     X=&Multiply_Float(
      A=35
      B=&Lerp_Float(A=1, B=2, Alpha=&FloatFromCurve(FloatCurve=$Emitter.FormingCurve, CurveIndex=$System.Forming))
     )
     Y=&Multiply_Float(
      A=35
      B=&Lerp_Float(
       A=1
       B=3
       Alpha=&FloatFromCurve(
        FloatCurve=DataInterface() {propertyValues {Curve {keys [
            {interpMode=RCIM_Cubic, tangentMode=RCTM_User, time=0, value=0}
            {interpMode=RCIM_Cubic, tangentMode=RCTM_User, time=0.5, value=1}
            {tangentMode=RCTM_User, time=1, value=0}
           ]}}}
        CurveIndex=$System.Forming
       )
      )
     )
    )
    Sprite Size Mode=::NewEnumerator4
    Sprite UV Mode=::NewEnumerator1
    UsePositionOffset=false
    Write Material Random=true
   )
   &CameraOffset(Camera Offset Amount=&SelectFloatFromArray(Direct Array Index=&SpawnGroup(), Float Selection Array=DataInterface() {propertyValues {FloatData [10,10]}}))
   &ShapeLocation(
    MaskSpawnGroup=false
    Override Seed=false
    Sphere Radius=&Multiply_Float(A=50, B=$System.FX_Scalar)
    UseOwnerScale=true
   )
   &AddVelocity(
    Module.Use Falloff Distance=false
    Override Seed=false
    Velocity Mode=::NewEnumerator1
    Velocity Speed=&Multiply_Float(
     A=&Multiply_Float(
      A=100
      B=&Lerp_Float(A=1, B=5, Alpha=&FloatFromCurve(FloatCurve=$Emitter.FormingCurve, CurveIndex=$System.Forming))
     )
     B=$System.FX_Scalar
    )
   )
  ]
  particleUpdateScript [
   &ParticleState()
   &ScaleSpriteSize(
    Scale Factor=&Multiply_Vector2D(A=&RandomRangeVector2D(Minimum=<2 2>, Maximum=<3 3>), B=<1 1>)
    Scale Sprite Size Mode=::NewEnumerator2
   ) {enabled=false}
   &ScaleSpriteSize(
    Non-Uniform Curve Scale=<3.5 3.5>
    Non-Uniform Curve Sprite Scale=DataInterface() {propertyValues {
      XCurve {keys [
        {arriveTangent=0.8609, interpMode=RCIM_Cubic, leaveTangent=0.8609, tangentMode=RCTM_User, time=0, value=0.6}
        {arriveTangent=0.5103, interpMode=RCIM_Cubic, leaveTangent=0.5103, tangentMode=RCTM_User, time=0.3, value=0.8}
        {arriveTangent=0.0224, interpMode=RCIM_Cubic, leaveTangent=0.0224, tangentMode=RCTM_User, time=1, value=1}
       ]}
      YCurve {keys [
        {arriveTangent=0.8801, interpMode=RCIM_Cubic, leaveTangent=0.8801, tangentMode=RCTM_User, time=0, value=0.3}
        {arriveTangent=0.6586, interpMode=RCIM_Cubic, leaveTangent=0.6586, tangentMode=RCTM_User, time=0.2, value=0.85}
        {arriveTangent=-0.4778, interpMode=RCIM_Cubic, leaveTangent=-0.4778, tangentMode=RCTM_User, time=1, value=0.75}
       ]}
     }}
    Scale Sprite Size Mode=::NewEnumerator3
   ) {module=ScaleSpriteSize001}
   &Drag(Drag=4, Use Linear Drag=true, Use Rotational Drag=true)
   &ApplyOwnerScaleToAttributes(Scale Initial Sprite Size=true)
   &SolveForcesAndVelocity(Clamp Velocity=false, Limit Acceleration=false)
   &ScaleColor(
    Scale Alpha=&Multiply_Float(
     A=&FloatFromCurve(
      FloatCurve=DataInterface() {propertyValues {Curve {keys [
          {arriveTangent=9.0725, interpMode=RCIM_Cubic, leaveTangent=9.0725, tangentMode=RCTM_User, time=0, value=1}
          {arriveTangent=-2.5249, interpMode=RCIM_Cubic, leaveTangent=-2.5249, tangentMode=RCTM_User, time=0.1408, value=0.9847}
          {arriveTangent=-0.6199, interpMode=RCIM_Cubic, leaveTangent=-0.6199, time=0.3, value=0.95}
          {arriveTangent=-1.3571, interpMode=RCIM_Cubic, leaveTangent=-1.3571, time=0.6, value=0.7}
          {arriveTangent=-0.0405, interpMode=RCIM_Cubic, leaveTangent=-0.0405, tangentMode=RCTM_User, time=1, value=0}
         ]}}}
      Scale Curve=2
     )
     B=1
    )
    Scale RGB=&VectorFromFloat(Value=0.05)
    ScaleA=true
    ScaleRGB=true
   ) {module=ScaleColor001}
   &ScaleColor(
    Scale Mode=::NewEnumerator1
    Scale RGBA=&MakeFloat4FromVectorAndFloat("Vector (RGB)"=<1 1 1>, "Float (Alpha)"=1)
    ScaleRGBA=true
   ) {enabled=false, module=ScaleColor002}
   &DynamicMaterialParameters(
    Index 0 Param 1=&FloatFromCurve(FloatCurve=DataInterface() {propertyValues {Curve {keys [
         {arriveTangent=-0.1846, time=0, value=0}
         {arriveTangent=1.3385, interpMode=RCIM_Cubic, leaveTangent=0.1195, tangentMode=RCTM_Break, time=0.3, value=0}
         {arriveTangent=1.4286, interpMode=RCIM_Cubic, leaveTangent=1.4286, time=0.8, value=0.5}
         {arriveTangent=3.0091, interpMode=RCIM_Cubic, leaveTangent=3.0091, tangentMode=RCTM_User, time=1, value=1}
        ]}}})
    Index 0 Param 2=&MakeCustomFloatFromBool(Bool=$User.IsUltra, True Float=0, False Float=0)
    Index 0 Param 4=20
    Module.Index0Param0Write=true
    Module.Index0Param1Write=true
    Module.Index0Param2Write=false
    Module.Index0Param3Write=false
   )
   /SetVariables(Particles.SpriteSize=&Multiply_Vector2D_ByFloat(Float=$System.FX_Scalar, Vector2D=$Particles.SpriteSize)) {isSetParameters=true, module=SetVariables_#5}
  ]
  rendererClasses [&NiagaraSpriteRendererProperties]
  renderers [{rendererClass=&NiagaraSpriteRendererProperties}]
 }
 SwirlsTrail {
  bEnabled=false
  emitterUpdateScript [
   &EmitterState()
   &SpawnPerUnit(
    Max Movement Threshold=10000
    Module.Use Max Movement Threshold=false
    Module.Use Movement Tolerance=false
    Spawn Probability:NiagaraFloat=&MakeCustomFloatFromBool(Bool=$User.IsSummon, True Float=0, False Float=1)
    Spawn Spacing=15
    Use Spawn Probability=true
   )
  ]
  particleSpawnScript [
   &PartitionParticles(Number Of Partitions=3, Write Partition to Ribbon ID=::NewEnumerator0)
   &InitializeParticle(
    Lifetime=&Multiply_Float(
     A=&SelectFloatFromArray(
      Direct Array Index=$Particles.PartitionParticles.Partition
      Float Selection Array=DataInterface() {propertyValues {FloatData [0.15,0.17,0.2]}}
     )
     B=1.5
    )
    Ribbon Width=&Multiply_Float(A=100, B=$System.FX_Scalar)
    Ribbon Width Mode=::NewEnumerator1
    UsePositionOffset=false
    Write Material Random=true
   )
   &ShapeLocation(
    Cylinder Distribution=::NewEnumerator2
    Cylinder Height=10
    Cylinder Height Midpoint=1
    Cylinder Radius=&Multiply_Float(
     A=&Multiply_Float(A=150, B=&Multiply_Float(A=$System.FX_Scalar, B=1))
     B=&SelectFloatFromArray(Direct Array Index=$Particles.PartitionParticles.Partition, Float Selection Array=DataInterface() {propertyValues {FloatData [0.7,0.9,1.1]}})
    )
    MaskSpawnGroup=false
    Offset=&MakeVector(X=0, Y=0, Z=0)
    Offset Mode=::NewEnumerator0
    Rotation Angle:NiagaraFloat=&Multiply_Float(
     A=$System.Age
     B=&SelectFloatFromArray(Direct Array Index=$Particles.PartitionParticles.Partition, Float Selection Array=DataInterface() {propertyValues {FloatData [1.2,2.2,3]}})
    )
    Rotation Mode=::NewEnumerator1
    Shape Origin=$Engine.Emitter.SimulationPosition
    Shape Primitive=::NewEnumerator1
    U Position=&Add_Float(A=0, B=&Multiply_FloatByInt(Float=0.33, Integer=$Particles.PartitionParticles.Partition))
    UseOwnerScale=true
   )
  ]
  particleUpdateScript [
   &ParticleState()
   &ApplyOwnerScaleToAttributes(Scale Ribbon Width=true)
   &ScaleRibbonWidth(Ribbon Width Scale=&Multiply_Float(
     A=1
     B=&FloatFromCurve(FloatCurve=DataInterface() {propertyValues {Curve {keys [
          {tangentMode=RCTM_User, time=0, value=1}
          {arriveTangent=-0.0405, interpMode=RCIM_Cubic, leaveTangent=-0.0405, tangentMode=RCTM_User, time=0.5, value=1}
          {arriveTangent=-4.065, interpMode=RCIM_Cubic, leaveTangent=-4.065, tangentMode=RCTM_User, time=1, value=0}
         ]}}})
    ))
   &DynamicMaterialParameters(
    Index 0 Param 1=&Lerp_Float(
     A=0
     B=1
     Alpha=&FloatFromCurve(FloatCurve=DataInterface() {propertyValues {Curve {keys [
          {arriveTangent=2.9487, interpMode=RCIM_Cubic, leaveTangent=2.9487, tangentMode=RCTM_User, time=0, value=0}
          {interpMode=RCIM_Cubic, tangentMode=RCTM_User, time=1, value=1}
         ]}}})
    )
    Index 0 Param 2=&Multiply_FloatByInt(Float=0.5, Integer=$Particles.PartitionParticles.Partition)
    Index 0 Param 4=1
    Module.Param0WriteEnabled=true
    Module.Param1WriteEnabled=true
    Module.Param2WriteEnabled=true
    Module.Param3WriteEnabled=false
   )
   &ScaleColor(
    Linear Color Curve=DataInterface() {propertyValues {
      AlphaCurve {keys [<at 0.8425 1>,<at 1 0>]}
      BlueCurve {keys [<at 0 0.2935>,<at 0.1781 1>,<at 0.3979 0.1128>]}
      GreenCurve {keys [<at 0 0.688>,<at 0.1781 1>,<at 0.3979 0.1128>]}
      RedCurve {keys [<at 0 1>,<at 0.1781 1>,<at 0.3979 0.1128>]}
     }}
    Scale Mode=::NewEnumerator2
   ) {module=ScaleColor002}
   &SolveForcesAndVelocity(Clamp Velocity=false, Limit Acceleration=false)
   &ScaleColor(
    Scale Alpha=&FloatFromCurve(
     FloatCurve=DataInterface() {propertyValues {Curve {keys [{interpMode=RCIM_Cubic, tangentMode=RCTM_User, time=0, value=1}, {interpMode=RCIM_Cubic, tangentMode=RCTM_User, time=1, value=0}]}}}
     Scale Curve=0.5
    )
    Scale RGB=&VectorFromFloat(Value=1)
    ScaleA=true
    ScaleRGB=true
   ) {module=ScaleColor003}
  ]
  rendererClasses [&NiagaraRibbonRendererProperties]
  renderers [{rendererClass=&NiagaraRibbonRendererProperties}]
 }
}
summary {
 systemName=NS_LimeRock_EnergyBall
 emitters [
  {emitterName=MainSphere, rendererClasses [&NiagaraMeshRendererProperties]}
  {bEnabled=false, emitterName=BlackDots, simTarget=GPUComputeSim, rendererClasses [&NiagaraSpriteRendererProperties]}
  {emitterName=StreaksDark, rendererClasses [&NiagaraSpriteRendererProperties]}
  {emitterName=LightAndGlow, rendererClasses [&NiagaraSpriteRendererProperties, &NiagaraLightRendererProperties]}
  {emitterName=StreaksBright, rendererClasses [&NiagaraSpriteRendererProperties]}
  {emitterName=PlasmaSphere, rendererClasses [&NiagaraMeshRendererProperties]}
  {emitterName=EnergySplineBolts001, rendererClasses [&NiagaraMeshRendererProperties]}
  {bEnabled=false, emitterName=SwirlsTrail, rendererClasses [&NiagaraRibbonRendererProperties]}
  {emitterName=BlackDots001, simTarget=GPUComputeSim, rendererClasses [&NiagaraSpriteRendererProperties]}
  {
   emitterName=LightAndGlow001
   rendererClasses [&NiagaraSpriteRendererProperties, &NiagaraLightRendererProperties]
  }
  {emitterName=EnergySplineBolts, rendererClasses [&NiagaraMeshRendererProperties]}
 ]
 userVariables [
  User.FX_Scalar {type=NiagaraFloat, defaultValue=2, description=""}
  User.IsSummon {type=NiagaraBool, defaultValue=true}
  User.IsUltra {type=NiagaraBool, defaultValue=true}
 ]
}
